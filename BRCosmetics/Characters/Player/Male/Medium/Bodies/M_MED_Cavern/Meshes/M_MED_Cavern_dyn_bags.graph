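# Physics asset for dangling bags, rope and grenades on the medium male Cavern player body cosmetic.
# Each bag and grenade swings within narrow angular limits from a tiny MSTR anchor body, which joins spine_01 with free rotation.
# A two-segment rope chains off the upper left bag, and a ring hangs from each grenade.
# Pelvis, spine_01 and left thigh capsules supply the body shapes, with no right thigh body.
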PhysicsAsset /BRCosmetics/Characters/Player/Male/Medium/Bodies/M_MED_Cavern/Meshes/M_MED_Cavern_dyn_bags
bodies {
 dyn_bag_01_l [{shapeType=Box, extentX=13.5, extentY=7, extentZ=4.5, center=<4.4252 0.0001 2.466>}]
 dyn_bag_01_l_MSTR [{shapeType=Capsule, length=0.505, radius=0.505, rotation=<rot 0 -90 0>}]
 dyn_bag_01_r [{shapeType=Box, extentX=12, extentY=7, extentZ=4.5, center=<4.5 0 2.2>}]
 dyn_bag_01_r_MSTR [{shapeType=Capsule, length=0.505, radius=0.505, rotation=<rot 0 -90 0>}]
 dyn_bag_02_l [{shapeType=Box, extentX=12, extentY=7, extentZ=5, center=<4.5 0 2.2>}]
 dyn_bag_02_l_MSTR [{shapeType=Capsule, length=0.505, radius=0.505, rotation=<rot 0 -90 0>}]
 dyn_bag_02_r [{shapeType=Box, extentX=12, extentY=7, extentZ=5, center=<4.5 0 2.2>}]
 dyn_bag_02_r_MSTR [{shapeType=Capsule, length=0.505, radius=0.505, rotation=<rot 0 -90 0>}]
 dyn_bag_bk [{shapeType=Box, extentX=12, extentY=19, extentZ=4.5, center=<3.9148 -0.051 2.9423>}]
 dyn_bag_bk_MSTR [{shapeType=Capsule, length=0.505, radius=0.505, rotation=<rot 0 -90 0>}]
 dyn_grenades_01 [{shapeType=Capsule, length=4.3371, radius=3.5328, center=<1 0 2>, rotation=<rot 0 85.7893 -87.8294>}]
 dyn_grenades_01_MSTR [{shapeType=Capsule, length=0.505, radius=0.505, rotation=<rot 0 -90 0>}]
 dyn_grenades_02 [{shapeType=Capsule, length=4.3371, radius=3.5899, center=<1 0 2>, rotation=<rot 0 84.9949 -88.5243>}]
 dyn_grenades_02_MSTR [{shapeType=Capsule, length=0.505, radius=0.505, rotation=<rot 0 -90 0>}]
 dyn_ring_01 [{shapeType=Capsule, length=1, radius=1, center=<1 0 0>, rotation=<rot 90 0 0>}]
 dyn_ring_02 [{shapeType=Capsule, length=1, radius=1, center=<1 0 0>, rotation=<rot 90 0 0>}]
 dyn_rope_01 [{shapeType=Box, extentX=8.5, extentY=17.5, extentZ=4.5, center=<3.6069 -0.4888 0.9306>}]
 dyn_rope_02 [{shapeType=Box, extentX=14, extentY=17.5, extentZ=4.5, center=<6.783 -0.4538 1.2895>, rotation=<rot 2.6041 0 0>}]
 pelvis [{shapeType=Capsule, length=6.3224, radius=12.5065, center=<0 -2.0707 0.0001>}]
 spine_01 [{shapeType=Capsule, length=6.75, radius=10.8395, center=<-0.0002 0.3555 0>}]
 thigh_l [
  {shapeType=Capsule, length=24.9014, radius=11.2583, center=<-11.9958 -0.5586 2.4811>, rotation=<rot 88.1666 179.9999 179.9999>}
  {shapeType=Capsule, length=21.5, radius=8.5, shapeName=thigh_l_capsule, center=<-15.3441 -6.6729 0>, rotation=<rot 90 0 0>}
 ]
}
constraints [
 {bone1Name=spine_01, bone2Name=pelvis, swing1LimitDegrees=45, swing2LimitDegrees=45, twistLimitDegrees=45}
 {
  bone1Name=dyn_bag_01_r_MSTR
  bone2Name=spine_01
  swing1LimitDegrees=45
  swing1Motion=Free
  swing2LimitDegrees=45
  swing2Motion=Free
  twistLimitDegrees=45
  twistMotion=Free
 }
 {bone1Name=dyn_bag_01_r, bone2Name=dyn_bag_01_r_MSTR, swing1LimitDegrees=10, swing2LimitDegrees=20, twistLimitDegrees=2}
 {
  bone1Name=dyn_bag_02_r_MSTR
  bone2Name=spine_01
  swing1LimitDegrees=45
  swing1Motion=Free
  swing2LimitDegrees=45
  swing2Motion=Free
  twistLimitDegrees=45
  twistMotion=Free
 }
 {bone1Name=dyn_bag_02_r, bone2Name=dyn_bag_02_r_MSTR, swing1LimitDegrees=10, swing2LimitDegrees=20, twistLimitDegrees=2}
 {
  bone1Name=dyn_bag_01_l_MSTR
  bone2Name=spine_01
  swing1LimitDegrees=45
  swing1Motion=Free
  swing2LimitDegrees=45
  swing2Motion=Free
  twistLimitDegrees=45
  twistMotion=Free
 }
 {bone1Name=dyn_bag_01_l, bone2Name=dyn_bag_01_l_MSTR, swing1LimitDegrees=10, swing2LimitDegrees=25, twistLimitDegrees=5}
 {
  bone1Name=dyn_bag_02_l_MSTR
  bone2Name=spine_01
  swing1LimitDegrees=45
  swing1Motion=Free
  swing2LimitDegrees=45
  swing2Motion=Free
  twistLimitDegrees=45
  twistMotion=Free
 }
 {bone1Name=dyn_bag_02_l, bone2Name=dyn_bag_02_l_MSTR, swing1LimitDegrees=10, swing2LimitDegrees=20, twistLimitDegrees=2}
 {
  bone1Name=dyn_bag_bk_MSTR
  bone2Name=spine_01
  swing1LimitDegrees=45
  swing1Motion=Free
  swing2LimitDegrees=45
  swing2Motion=Free
  twistLimitDegrees=45
  twistMotion=Free
 }
 {bone1Name=dyn_bag_bk, bone2Name=dyn_bag_bk_MSTR, swing1LimitDegrees=2, swing2LimitDegrees=20, twistLimitDegrees=2}
 {bone1Name=dyn_rope_01, bone2Name=dyn_bag_01_l, swing1LimitDegrees=25, swing2LimitDegrees=25, twistLimitDegrees=10}
 {bone1Name=dyn_rope_02, bone2Name=dyn_rope_01, swing1LimitDegrees=25, swing2LimitDegrees=25, twistLimitDegrees=10}
 {
  bone1Name=thigh_l
  bone2Name=pelvis
  swing1LimitDegrees=45
  swing1Motion=Free
  swing2LimitDegrees=45
  swing2Motion=Free
  twistLimitDegrees=45
  twistMotion=Free
 }
 {
  bone1Name=dyn_grenades_01_MSTR
  bone2Name=spine_01
  swing1LimitDegrees=45
  swing1Motion=Free
  swing2LimitDegrees=45
  swing2Motion=Free
  twistLimitDegrees=45
  twistMotion=Free
 }
 {bone1Name=dyn_grenades_01, bone2Name=dyn_grenades_01_MSTR, swing1LimitDegrees=20, swing2LimitDegrees=20, twistLimitDegrees=10}
 {bone1Name=dyn_ring_01, bone2Name=dyn_grenades_01, swing1LimitDegrees=5, swing2LimitDegrees=30, twistLimitDegrees=2}
 {
  bone1Name=dyn_grenades_02_MSTR
  bone2Name=spine_01
  swing1LimitDegrees=45
  swing1Motion=Free
  swing2LimitDegrees=45
  swing2Motion=Free
  twistLimitDegrees=45
  twistMotion=Free
 }
 {bone1Name=dyn_grenades_02, bone2Name=dyn_grenades_02_MSTR, swing1LimitDegrees=20, swing2LimitDegrees=20, twistLimitDegrees=10}
 {bone1Name=dyn_ring_02, bone2Name=dyn_grenades_02, swing1LimitDegrees=5, swing2LimitDegrees=30, twistLimitDegrees=2}
]
