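# Animation Blueprint for the M_MED_Cavern player body cosmetic, adding deformation and secondary motion to bags, rope and grenades on a pose copied from the attached parent mesh.
# Two Control Rigs and weighted pelvis, spine and clavicle constraints build a cached pose.
# bRigidBodyEnabled selects either RigidBody bag simulation with left-thigh offsets, or a look-at and additive-offset fallback for left bag and rope.
# Update event derives alphas from thigh rotation; the right-thigh value is unused.
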AnimBlueprint /BRCosmetics/Characters/Player/Male/Medium/Bodies/M_MED_Cavern/Meshes/M_MED_Cavern_AnimBP
target_skeleton=&/M_MED_Cavern_Skeleton
graphs [
 AnimGraph {nodes [
   n1 = Get ropeOffsetAlpha() -> (ropeOffsetAlpha:double)
   n2 = Look At - Bone: dyn_bag_01_l(ComponentPose=Local To Component(LocalPose=Use cached pose 'perPose'()), Alpha=@n1) {
    alphaBoolBlend {blendOption=Linear}
    boneToModify {boneName=dyn_bag_01_l}
    lookAtLocation=<0 -3 -11>
    lookAtTarget {boneReference {boneName=thigh_twist_01_l}}
    lookAt_Axis {axis=<1 0>}
   }
   n3 = Look At - Bone: dyn_rope_01(ComponentPose=@n2, Alpha=@n1) {
    alphaBoolBlend {blendOption=Linear}
    boneToModify {boneName=dyn_rope_01}
    lookAtLocation=<0 -3 -11>
    lookAtTarget {boneReference {boneName=thigh_twist_01_l}}
    lookAt_Axis {axis=<1 0>}
   }
   n4 = Get thigh_rot_up_l() -> (thigh_rot_up_l:double)
   n5 = "Transform (Modify) Bone - Bone: dyn_bag_01_l"(ComponentPose=@n3, Alpha=@n4, Translation="0.000,-1.486,0.640", Rotation="0.000,0.000,-10.000", Scale="1,1,1") {
    rotationMode=BMM_Additive
    rotationSpace=BCS_BoneSpace
    translationMode=BMM_Additive
    translationSpace=BCS_BoneSpace
    alphaBoolBlend {blendOption=Linear}
    boneToModify {boneName=dyn_bag_01_l}
    rotation {roll=-10.0002}
    translation {y=-1.486, z=0.64}
   }
   n6 = "Transform (Modify) Bone - Bone: dyn_rope_01"(ComponentPose=@n5, Alpha=@n4, Translation="-0.406,-0.802,0.844", Rotation="-4.849,0,-14.145", Scale="1,1,1") {
    rotationMode=BMM_Additive
    rotationSpace=BCS_BoneSpace
    scaleMode=BMM_Additive
    translationMode=BMM_Additive
    translationSpace=BCS_BoneSpace
    alphaBoolBlend {blendOption=Linear}
    boneToModify {boneName=dyn_rope_01}
    rotation {pitch=-4.849, roll=-14.145}
    translation=<-0.406 -0.802 0.844>
   }
   n7 = Get thigh_rot_up_l()
   n8 = "Transform (Modify) Bone - Bone: dyn_bag_01_l_MSTR"(
    ComponentPose=Local To Component(LocalPose=Use cached pose 'perPose'())
    Alpha=@n7
    Translation="-0.062,-3.067,-0.615"
    Rotation="17.415,-12.681,-10.243"
    Scale="1,1,1"
   ) {
    rotationMode=BMM_Additive
    rotationSpace=BCS_BoneSpace
    translationMode=BMM_Additive
    translationSpace=BCS_BoneSpace
    alphaBoolBlend {blendOption=Linear}
    boneToModify {boneName=dyn_bag_01_l_MSTR}
    rotation=<rot 17.4149 -12.6814 -10.2434>
    translation=<-0.062 -3.067 -0.615>
   }
   n9 = "Transform (Modify) Bone - Bone: dyn_bag_02_l_MSTR"(ComponentPose=@n8, Alpha=@n7, Translation="-0.044,-0.842,-0.004", Rotation="0,-6.002,-20", Scale="1,1,1") {
    rotationMode=BMM_Additive
    rotationSpace=BCS_BoneSpace
    translationMode=BMM_Additive
    translationSpace=BCS_BoneSpace
    alphaBoolBlend {blendOption=Linear}
    boneToModify {boneName=dyn_bag_02_l_MSTR}
    rotation {roll=-20, yaw=-6.002}
    translation=<-0.0441 -0.842 -0.004>
   }
   n10 = RigidBody(ComponentPose=@n9) {
    bClampLinearTranslationLimitToRefPose=true
    bOverrideWorldGravity=true
    overridePhysicsAsset=&/M_MED_Cavern_dyn_bags
    simulationSpace=BaseBoneSpace
    alphaBoolBlend {blendOption=Linear}
    baseBoneRef {boneName=pelvis}
    componentAppliedLinearAccClamp=<3000 3000 3000>
    componentLinearAccScale=<0.32 0.35 0.35>
    componentLinearVelScale=<0.3 0.3 0.3>
    overrideWorldGravity {z=-980}
   }
   Output Pose(Result=Blend Poses by bool(BlendPose_0=Component To Local(ComponentPose=@n10), BlendPose_1=Component To Local(ComponentPose=@n6), bActiveValue=Get bRigidBodyEnabled()) {blendType=Linear}) {layerGroup=None}
   n11 = Control Rig(Source=Copy Pose From Mesh() {bUseAttachedParent=true}) {
    alphaBoolBlend {blendOption=Linear}
    controlRigAssetReference {blueprintRigClass=&/M_MED_Cavern_Deform_CtrlRig_twists.M_MED_Cavern_Deform_CtrlRig_twists_C}
   }
   n12 = Control Rig!(
    Source=@n11
    Alpha
    KneeRot_fr_l_114deg:Vector
    KneeTrans_fr_l_114deg:Vector="0.258,-3.591,0.009"
    KneeScale_fr_l_114deg:Vector="1,1,1"
    KneeScale_fr_l_130deg:Vector="1,1,1"
    KneeRot_fr_l_130deg:Vector="0,0,-51.2"
    KneeTrans_fr_l_130deg:Vector="0.258,-3.591,0.009"
   ) -> (Pose) {
    alphaBoolBlend {blendOption=Linear}
    controlRigAssetReference {blueprintRigClass=&/M_MED_Cavern_Deform_CtrlRig_sculpters.M_MED_Cavern_Deform_CtrlRig_sculpters_C}
   }
   n13 = "Constraint - dyn_bag_01_r_MSTR (PPPP)"!(
    ComponentPose=Local To Component(LocalPose=@n12)
    Alpha=1
    ConstraintWeights_0=0.6673
    ConstraintWeights_1:float=0.0915
    ConstraintWeights_2:float=0.0282
    ConstraintWeights_3:float=0.2129
   ) -> (Pose) {
    alphaBoolBlend {blendOption=Linear}
    boneToModify {boneName=dyn_bag_01_r_MSTR}
    constraintSetup [{targetBone {boneName=pelvis}}, {targetBone {boneName=spine_02}}, {targetBone {boneName=spine_03}}, {targetBone {boneName=spine_01}}]
    constraintWeights [1,1,1,1]
   }
   n14 = "Constraint - dyn_bag_02_r_MSTR (PPPP)"!(
    ComponentPose=@n13
    Alpha=1
    ConstraintWeights_0=0.6673
    ConstraintWeights_1:float=0.0915
    ConstraintWeights_2:float=0.0282
    ConstraintWeights_3:float=0.2129
   ) -> (Pose) {
    alphaBoolBlend {blendOption=Linear}
    boneToModify {boneName=dyn_bag_02_r_MSTR}
    constraintSetup [{targetBone {boneName=pelvis}}, {targetBone {boneName=spine_02}}, {targetBone {boneName=spine_03}}, {targetBone {boneName=spine_01}}]
    constraintWeights [1,1,1,1]
   }
   n15 = "Constraint - dyn_bag_01_l_MSTR (PPPP)"!(
    ComponentPose=@n14
    Alpha=1
    ConstraintWeights_0=0.6673
    ConstraintWeights_1:float=0.0915
    ConstraintWeights_2:float=0.0282
    ConstraintWeights_3:float=0.2129
   ) -> (Pose) {
    alphaBoolBlend {blendOption=Linear}
    boneToModify {boneName=dyn_bag_01_l_MSTR}
    constraintSetup [{targetBone {boneName=pelvis}}, {targetBone {boneName=spine_02}}, {targetBone {boneName=spine_03}}, {targetBone {boneName=spine_01}}]
    constraintWeights [1,1,1,1]
   }
   n16 = "Constraint - dyn_bag_02_l_MSTR (PPPP)"!(
    ComponentPose=@n15
    Alpha=1
    ConstraintWeights_0=0.6673
    ConstraintWeights_1:float=0.0915
    ConstraintWeights_2:float=0.0282
    ConstraintWeights_3:float=0.2129
   ) -> (Pose) {
    alphaBoolBlend {blendOption=Linear}
    boneToModify {boneName=dyn_bag_02_l_MSTR}
    constraintSetup [{targetBone {boneName=pelvis}}, {targetBone {boneName=spine_02}}, {targetBone {boneName=spine_03}}, {targetBone {boneName=spine_01}}]
    constraintWeights [1,1,1,1]
   }
   n17 = "Constraint - dyn_bag_bk_MSTR (PPPP)"!(
    ComponentPose=@n16
    Alpha=1
    ConstraintWeights_0=0.6673
    ConstraintWeights_1:float=0.0915
    ConstraintWeights_2:float=0.0282
    ConstraintWeights_3:float=0.2129
   ) -> (Pose) {
    alphaBoolBlend {blendOption=Linear}
    boneToModify {boneName=dyn_bag_bk_MSTR}
    constraintSetup [{targetBone {boneName=pelvis}}, {targetBone {boneName=spine_02}}, {targetBone {boneName=spine_03}}, {targetBone {boneName=spine_01}}]
    constraintWeights [1,1,1,1]
   }
   n18 = "Constraint - dyn_grenades_01_MSTR (PPPPPP)"!(
    ComponentPose=@n17
    Alpha=1
    ConstraintWeights_0=0.1067
    ConstraintWeights_1:float=0.229
    ConstraintWeights_2:float=0.4652
    ConstraintWeights_3:float=0.1483
    ConstraintWeights_4:float=0.0003
    ConstraintWeights_5:float=0.0504
   ) -> (Pose) {
    alphaBoolBlend {blendOption=Linear}
    boneToModify {boneName=dyn_grenades_01_MSTR}
    constraintSetup [
     {targetBone {boneName=spine_01}}
     {targetBone {boneName=spine_02}}
     {targetBone {boneName=spine_03}}
     {targetBone {boneName=spine_04}}
     {targetBone {boneName=spine_05}}
     {targetBone {boneName=clavicle_r}}
    ]
    constraintWeights [1,1,1,1,1,1]
   }
   n19 = "Constraint - dyn_grenades_02_MSTR (PPPPPPP)"!(
    ComponentPose=@n18
    Alpha=1
    ConstraintWeights_0=0.1246
    ConstraintWeights_1:float=0.2914
    ConstraintWeights_2:float=0.4093
    ConstraintWeights_3:float=0.0908
    ConstraintWeights_4:float=0
    ConstraintWeights_5:float=0.0356
    ConstraintWeights_6:float=0.0484
   ) -> (Pose) {
    alphaBoolBlend {blendOption=Linear}
    boneToModify {boneName=dyn_grenades_02_MSTR}
    constraintSetup [
     {targetBone {boneName=spine_01}}
     {targetBone {boneName=spine_02}}
     {targetBone {boneName=spine_03}}
     {targetBone {boneName=spine_04}}
     {targetBone {boneName=spine_05}}
     {targetBone {boneName=clavicle_r}}
     {targetBone {boneName=pelvis}}
    ]
    constraintWeights [1,1,1,1,1,1,1]
   }
   Save cached pose 'perPose'(Pose=Component To Local(ComponentPose=@n19)) {cacheName=perPose}
  ]}
 EventGraph {nodes [
   Event BlueprintUpdateAnimation()
   n1 = BreakTransform(InTransform=GetDeltaTransformFromRefPose(self=GetOwnerSkeletalMeshComponent(), BoneName=thigh_r, BaseName=pelvis))
   > Set thigh_rot_up_r(thigh_rot_up_r:double=MapRangeClamped(Value=BreakRotator(InRot=@n1.Rotation).Yaw, InRangeA=-15, InRangeB=-100)) -> (Output_Get:double)
   n2 = BreakTransform(InTransform=GetDeltaTransformFromRefPose(self=GetOwnerSkeletalMeshComponent(), BoneName=thigh_l, BaseName=pelvis))
   > Set thigh_rot_up_l(thigh_rot_up_l:double=MapRangeClamped(Value=BreakRotator(InRot=@n2.Rotation).Yaw, InRangeA=0, InRangeB=-100)) -> (Output_Get:double)
   n3 = NOT Boolean(A=Get bRigidBodyEnabled())
   n4 = Get thigh_rot_up_l()
   > Set ropeOffsetAlpha(ropeOffsetAlpha:double=float * float(A=MapRangeClamped(Value=@n4, InRangeA=0, InRangeB=1, OutRangeA=1, OutRangeB=0), B="To Float (Boolean)"(InBool=@n3))) -> (Output_Get:double)
  ]}
]
variables [
 ropeOffsetAlpha {type=number}
 thigh_rot_up_l {type=number}
 thigh_rot_up_r {type=number}
]
