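# Physics asset of cape colliders for the M_MED_Cavern medium male player body, by its name giving cloth shapes to collide against.
# Capsules cover the pelvis, spine, clavicles and neck, with a sphere on the head and an extra capsule on the dyn_rope_02 bone.
# Spine_02 adds two long, mirrored side capsules to its main one.
# Arm, leg and foot bodies hold no shapes, and no constraints link any bodies.
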PhysicsAsset /BRCosmetics/Characters/Player/Male/Medium/Bodies/M_MED_Cavern/Meshes/M_MED_Cavern_CapeColliders
bodies {
 calf_l []
 calf_r []
 clavicle_l [{shapeType=Capsule, length=7.0136, radius=11.8927, center=<10.9017 1.1306 -2.615>, rotation=<rot 67.1516 3.7279 174.1089>}]
 clavicle_r [{shapeType=Capsule, length=5.6386, radius=12.5177, center=<-16.0964 -0.7029 2.2072>, rotation=<rot 75.9351 -0.0002 -0.912>}]
 dyn_rope_02 [{shapeType=Capsule, length=8, radius=13, shapeName=dyn_rope_02_capsule, center=<-0.013 0 -0.0681>, rotation=<rot -90 179.9963 -179.9961>}]
 foot_l []
 foot_r []
 head [{shapeType=Sphere, radius=17.3125, center=<11.028 -0.174 0>}]
 lowerarm_l []
 lowerarm_r []
 neck_02 [{shapeType=Capsule, length=13.4375, radius=7.3175, center=<-0.2642 -0.3408 0>, rotation=<rot 90 -171.1193 -175.3978>}]
 pelvis [{shapeType=Capsule, length=9.9062, radius=19.0388, center=<-2.6215 -2.053 0>}]
 spine_01 [{shapeType=Capsule, length=10.2187, radius=17.6638, center=<-2.6215 -2.0823 0>}]
 spine_02 [
  {shapeType=Capsule, length=10.0937, radius=15.195, center=<-1.7481 -1.9934 0.0001>}
  {shapeType=Capsule, length=29.875, radius=11.1013, center=<2.5024 0.4992 3.9825>, rotation=<rot 90 0.0049 0.0049>}
  {shapeType=Capsule, length=29.875, radius=11.1013, center=<2.4704 0.4991 -3.9825>, rotation=<rot 90 -0.0396 -0.0396>}
 ]
 spine_03 [{shapeType=Capsule, length=10.9062, radius=14.0075, center=<-1.953 -2.8111 0.0001>}]
 spine_04 [{shapeType=Capsule, length=14.4062, radius=13.8825, center=<2.1467 -1.9155 0.0002>}]
 spine_05 [{shapeType=Capsule, length=12.7812, radius=17.0625, center=<-4.1426 2.395 0>}]
 thigh_l []
 thigh_r []
 upperarm_l []
 upperarm_r []
}
constraints []
